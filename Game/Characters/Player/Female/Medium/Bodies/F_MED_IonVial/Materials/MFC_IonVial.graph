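# Material function for the IonVial player character that recolors masked areas of incoming material attributes with a custom color.
# A custom texture or input mask is channel-selected and power-shaped, optionally adds the texture's red channel, and has a raw and Fresnel-attenuated version.
# These masks blend base color, emissive, metallic, specular and roughness toward IonVial parameters.
# The blended emissive passes through the shared BoostEmissivebyTOD function before output.
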MaterialFunction /Game/Characters/Player/Female/Medium/Bodies/F_MED_IonVial/Materials/MFC_IonVial
graph {nodes [
  n1 = GetMaterialAttributes(None=FunctionInput() {inputName=MA_IN, inputType=FunctionInput_MaterialAttributes}) {attributeGetTypes [
    #1
    #2
    #3
    #4
    #5
   ]}
  n2 = VectorParameter() {group=IonVial, parameterName=IonVial_CustomColor, defaultValue=<rgba 0 0.2661 0.9583 1>}
  n3 = ScalarParameter() {defaultValue=0.5, group=IonVial, parameterName=IonVial_BaseColorIntensity}
  n4 = NamedRerouteUsage()
  n5 = Multiply(A=@n2.RGB, B=ScalarParameter() {defaultValue=1.2, group=IonVial, parameterName=IonVial_CustomColorIntensity})
  n6 = NamedRerouteUsage()
  n7 = SetMaterialAttributes!(
   MaterialAttributes=@n1.MaterialAttributes
   BaseColor=LinearInterpolate(A=@n1.BaseColor, B=Multiply(A=@n2.RGB, B=@n3), Alpha=@n4)
   EmissiveColor=MaterialFunctionCall!(In=LinearInterpolate(A=@n1.EmissiveColor, B=@n5, Alpha=@n4)) {materialFunction=/Game/Weapons/WeaponSkins/MaterialFunction/BoostEmissivebyTOD}.Result
   Metallic=LinearInterpolate(A=@n1.Metallic, B=ScalarParameter() {group=IonVial, parameterName=IonVial_Metallic}, Alpha=@n6)
   Specular=LinearInterpolate(A=@n1.Specular, B=ScalarParameter() {defaultValue=0.8, group=IonVial, parameterName=IonVial_Specular}, Alpha=@n6)
   Roughness=LinearInterpolate(A=@n1.Roughness, B=ScalarParameter() {defaultValue=0.2, group=IonVial, parameterName=IonVial_Roughness}, Alpha=@n6)
  ) {attributeSetTypes [
    #1
    #2
    #3
    #4
    #5
   ]}
  FunctionOutput(None=@n7) {outputName=MA_OUT}
  n8 = Multiply(
   A=Fresnel() {baseReflectFraction=0, exponent=1}
   B=ScalarParameter() {defaultValue=0.9, group=IonVial, parameterName=IonVial_FresnelIntensity}
  )
  n9 = TextureSampleParameter2D() {group=IonVial, parameterName=IonVial_CustomColorMask, texture=&T_BlacksRGB}
  n10 = ChannelMaskParameter(None=StaticSwitchParameter(True=@n9.RGB, False=FunctionInput() {inputName=Mask, sortPriority=1}) {group=IonVial, parameterName=IonVial_UseCustomMask}) {group=IonVial, maskChannel=Alpha, parameterName=IonVial_CustomColorChannel, defaultValue=<rgba 0 0 0 1>}
  n11 = Saturate(None=Power(Base=@n10.R, Exp=ScalarParameter() {defaultValue=1, group=IonVial, parameterName=IonVial_CustomColorMask_Power}))
  n12 = StaticSwitchParameter(True=Add(A=@n9.R, B=@n11), False=@n11) {group=IonVial, parameterName=IonVial_UseAdditionalChannel}
  NamedRerouteDeclaration(None=Multiply(A=OneMinus(None=@n8), B=NamedRerouteDeclaration(None=@n12) {name=CustomColor_Mask_Raw, nodeColor=<rgba 1 0 0.4922 1>})) {name=CustomColor_Mask, nodeColor=<rgba 0 0.2969 1 1>}
 ]}
