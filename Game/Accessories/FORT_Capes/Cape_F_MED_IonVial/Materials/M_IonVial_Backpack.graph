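# Character material for the IonVial backpack cosmetic, built from material attributes and usable on skeletal, cloth and Nanite meshes.
# The shared PreFX character function feeds optional thin film and metals LUT layers behind static switches, both driven by the SkinFX_Mask texture.
# The IonVial-specific function, borrowed from the character body's folder, applies its effect with the same mask.
# The result then passes through the shared PostFX and cinematic PostFX functions.
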Material /Game/Accessories/FORT_Capes/Cape_F_MED_IonVial/Materials/M_IonVial_Backpack
graph {
 properties {
  bAutomaticallySetUsageInEditor=false
  bUseMaterialAttributes=true
  bUsedWithClothing=true
  bUsedWithMeshDeformer=true
  bUsedWithMorphTargets=true
  bUsedWithNanite=true
  bUsedWithSkeletalMesh=true
  pixelDepthOffsetMode=PDOM_Legacy
  refractionMethod=RM_IndexOfRefraction
  shadingModel=MSM_FromMaterialExpression
  subsurfaceProfile=&SS_HeroSkin_02
  usedShadingModels=Default Lit | Subsurface Profile
 }
 nodes [
  n1 = MaterialFunctionCall() {materialFunction=&BaseFN_CharacterFunction_PreFX}
  n2 = TextureSample(Tex=TextureObjectParameter() {group=Skin FX, parameterName=SkinFX_Mask})
  n3 = StaticSwitchParameter(
   True=MaterialFunctionCall!(
    Alpha=StaticComponentMaskParameter(None=@n2.RGBA) {defaultR=true, group=Skin FX, parameterName=ThinFilm_Channel}
    Material Attributes=@n1.Result
   ) {materialFunction=/Game/Weapons/WeaponSkins/MaterialFunction/ThinFilmFunction}.Result
   False=@n1.Result
  ) {group=Skin FX, parameterName=Use Thin Film}
  n4 = MaterialFunctionCall(Mask=@n2.RGBA, Materials=@n3) {materialFunction=&MF_MetalsLUT}
  n5 = MaterialFunctionCall(
   AO=@n1.AO
   Distance Blend=@n1.AO
   Fresnel=@n1.AO
   Gmap=@n1.AO
   Pre Skinned Local Position=@n1.AO
   Material=MaterialFunctionCall!(MA_IN=StaticSwitchParameter(True=@n4.Result, False=@n3) {group=Skin FX, parameterName=UseMetalLUT}, Mask=@n2.RGBA) {materialFunction=/Game/Characters/Player/Female/Medium/Bodies/F_MED_IonVial/Materials/MFC_IonVial}.MA_OUT
  ) {materialFunction=&BaseFN_CharacterFunction_PostFX}
  MaterialFunctionCall(In=@n5.Result) {materialFunction=&CINE_MF_CharPostFX}
 ]
}
